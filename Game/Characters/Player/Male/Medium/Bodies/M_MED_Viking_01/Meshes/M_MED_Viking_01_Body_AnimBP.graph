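# Post-process-style body AnimBP for the male medium Viking 01 player character, adding secondary motion to capelet, bag, calf fur and skirt bones.
# It copies the attached parent's pose, then runs serial pelvis-relative AnimDynamics: low-alpha linear springs for capelet and fur, angular springs for bag and skirt chains.
# Fortnite Flap nodes scale back skirt bones from calf rotation.
# The sequence player and two rotation Flap nodes are unconnected; EventGraph has no logic.
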AnimBlueprint /Game/Characters/Player/Male/Medium/Bodies/M_MED_Viking_01/Meshes/M_MED_Viking_01_Body_AnimBP
target_skeleton=&/M_MED_Viking_01_Body_Skeleton
graphs [
 AnimGraph {nodes [
   n1 = Anim Dynamics - Bone: dyn_capelet_l(ComponentPose=Local To Component(LocalPose=Copy Pose From Mesh() {bUseAttachedParent=true}), Alpha=0.08, ExternalForce="0,0,0") {
    bGravityOverrideInSimSpace=true
    bLinearSpring=true
    bOverrideLinearDamping=true
    gravityScale=0.2
    lODThreshold=1
    linearDampingOverride=0.99
    linearSpringConstant=350
    alphaBoolBlend {blendOption=Linear}
    boundBone {boneName=dyn_capelet_l}
    physicsBodyDefinitions [{
      sphereCollisionRadius=0
      boundBone {boneName=dyn_capelet_l}
      boxExtents=<40 30 20>
      constraintSetup {
       twistAxis=AxisZ
       angularLimitsMax=<0 0 0>
       angularLimitsMin=<0 0 0>
       angularTarget=<0 0 0>
       linearAxesMax=<10 10 10>
       linearAxesMin=<-10 -10 -10>
      }
      localJointOffset=<0 0 0>
     }]
    relativeSpaceBone {boneName=pelvis}
   }
   n2 = Anim Dynamics - Bone: dyn_capelet_r(ComponentPose=@n1, Alpha=0.08, ExternalForce="0,0,0") {
    bGravityOverrideInSimSpace=true
    bLinearSpring=true
    bOverrideLinearDamping=true
    gravityScale=0.2
    lODThreshold=1
    linearDampingOverride=0.99
    linearSpringConstant=330
    alphaBoolBlend {blendOption=Linear}
    boundBone {boneName=dyn_capelet_r}
    physicsBodyDefinitions [{
      sphereCollisionRadius=0
      boundBone {boneName=dyn_capelet_r}
      boxExtents=<25 30 20>
      constraintSetup {
       twistAxis=AxisZ
       angularLimitsMax=<0 0 0>
       angularLimitsMin=<0 0 0>
       angularTarget=<0 0 0>
       linearAxesMax=<10 10 10>
       linearAxesMin=<-10 -10 -10>
      }
      localJointOffset=<0 0 0>
     }]
    relativeSpaceBone {boneName=pelvis}
   }
   n3 = Anim Dynamics - Bone: dyn_capelet_b(ComponentPose=@n2, Alpha=0.08, ExternalForce="0,0,0") {
    bGravityOverrideInSimSpace=true
    bLinearSpring=true
    bOverrideLinearDamping=true
    gravityScale=0.2
    lODThreshold=1
    linearDampingOverride=0.99
    linearSpringConstant=320
    alphaBoolBlend {blendOption=Linear}
    boundBone {boneName=dyn_capelet_b}
    physicsBodyDefinitions [{
      sphereCollisionRadius=0
      boundBone {boneName=dyn_capelet_b}
      boxExtents=<30 25 15>
      constraintSetup {
       twistAxis=AxisZ
       angularLimitsMax=<0 0 0>
       angularLimitsMin=<0 0 0>
       angularTarget=<0 0 0>
       linearAxesMax=<10 10 10>
       linearAxesMin=<-10 -10 -10>
      }
      localJointOffset=<0 0 0>
     }]
    relativeSpaceBone {boneName=pelvis}
   }
   n4 = Anim Dynamics - Bone: dyn_bag(ComponentPose=@n3, Alpha=1, ExternalForce="0,0,0") {
    angularDampingOverride=0.99
    angularSpringConstant=5000
    bAngularSpring=true
    bGravityOverrideInSimSpace=true
    bOverrideAngularDamping=true
    numSolverIterationsPostUpdate=4
    numSolverIterationsPreUpdate=16
    alphaBoolBlend {blendOption=Linear}
    boundBone {boneName=dyn_bag}
    physicsBodyDefinitions [{
      sphereCollisionRadius=0
      boundBone {boneName=dyn_bag}
      boxExtents=<10 20 20>
      constraintSetup {angularLimitsMax=<5 20 20>, angularLimitsMin=<-5 -20 -15>, angularTarget=<1 0 0>}
      localJointOffset=<3 0 0>
     }]
    relativeSpaceBone {boneName=pelvis}
   }
   n5 = Anim Dynamics - Bone: dyn_calf_fur_l(ComponentPose=@n4, Alpha=0.05, ExternalForce="0,0,0") {
    bGravityOverrideInSimSpace=true
    bLinearSpring=true
    bOverrideLinearDamping=true
    gravityScale=0.2
    lODThreshold=1
    linearDampingOverride=0.99
    linearSpringConstant=320
    alphaBoolBlend {blendOption=Linear}
    boundBone {boneName=dyn_calf_fur_l}
    physicsBodyDefinitions [{
      sphereCollisionRadius=0
      boundBone {boneName=dyn_calf_fur_l}
      boxExtents=<30 25 15>
      constraintSetup {twistAxis=AxisZ, angularLimitsMax=<0 0 0>, angularLimitsMin=<0 0 0>, angularTarget=<0 0 0>, linearAxesMax=<10 0 0>, linearAxesMin=<-10 0 0>}
      localJointOffset=<0 0 0>
     }]
    relativeSpaceBone {boneName=pelvis}
   }
   n6 = Anim Dynamics - Bone: dyn_calf_fur_r(ComponentPose=@n5, Alpha=0.05, ExternalForce="0,0,0") {
    bGravityOverrideInSimSpace=true
    bLinearSpring=true
    bOverrideLinearDamping=true
    gravityScale=0.2
    lODThreshold=1
    linearDampingOverride=0.99
    linearSpringConstant=320
    alphaBoolBlend {blendOption=Linear}
    boundBone {boneName=dyn_calf_fur_r}
    physicsBodyDefinitions [{
      sphereCollisionRadius=0
      boundBone {boneName=dyn_calf_fur_r}
      boxExtents=<30 25 15>
      constraintSetup {twistAxis=AxisZ, angularLimitsMax=<0 0 0>, angularLimitsMin=<0 0 0>, angularTarget=<0 0 0>, linearAxesMax=<10 0 0>, linearAxesMin=<-10 0 0>}
      localJointOffset=<0 0 0>
     }]
    relativeSpaceBone {boneName=pelvis}
   }
   n7 = "Anim Dynamics - Chain: dyn_skirt_out_l -> dyn_skirt_out_02_l"(ComponentPose=@n6, Alpha=1, ExternalForce="0,0,0") {
    angularDampingOverride=0.999
    angularSpringConstant=100
    bAngularSpring=true
    bChain=true
    bGravityOverrideInSimSpace=true
    bOverrideAngularDamping=true
    numSolverIterationsPostUpdate=4
    numSolverIterationsPreUpdate=16
    alphaBoolBlend {blendOption=Linear}
    boundBone {boneName=dyn_skirt_out_l}
    chainEnd {boneName=dyn_skirt_out_02_l}
    physicsBodyDefinitions [{
      sphereCollisionRadius=0
      boundBone {boneName=dyn_skirt_out_l}
      boxExtents=<50 50 30>
      constraintSetup {angularTargetAxis=AxisZ, twistAxis=AxisZ, angularLimitsMax=<10 0 5>, angularLimitsMin=<-10 -10 -5>, angularTarget=<0 0 1>}
      localJointOffset=<0 0 -2>
     }]
    relativeSpaceBone {boneName=pelvis}
   }
   n8 = "Anim Dynamics - Chain: dyn_skirt_fr_l -> dyn_skirt_fr_02_l"(ComponentPose=@n7, Alpha=1, ExternalForce="0,0,0") {
    angularDampingOverride=0.999
    angularSpringConstant=300
    bAngularSpring=true
    bChain=true
    bGravityOverrideInSimSpace=true
    bOverrideAngularDamping=true
    numSolverIterationsPostUpdate=4
    numSolverIterationsPreUpdate=16
    alphaBoolBlend {blendOption=Linear}
    boundBone {boneName=dyn_skirt_fr_l}
    chainEnd {boneName=dyn_skirt_fr_02_l}
    physicsBodyDefinitions [{
      sphereCollisionRadius=0
      boundBone {boneName=dyn_skirt_fr_l}
      boxExtents=<18 25 8>
      constraintSetup {angularTargetAxis=AxisZ, twistAxis=AxisZ, angularLimitsMax=<2 5 5>, angularLimitsMin=<-2 -20 -5>, angularTarget=<0 0 1>}
      localJointOffset=<0 0 -5>
     }]
    relativeSpaceBone {boneName=pelvis}
   }
   n9 = "Anim Dynamics - Chain: dyn_skirt_bk_l -> dyn_skirt_bk_02_l"(ComponentPose=@n8, Alpha=1, ExternalForce="0,0,0") {
    angularDampingOverride=0.999
    angularSpringConstant=300
    bAngularSpring=true
    bChain=true
    bGravityOverrideInSimSpace=true
    bOverrideAngularDamping=true
    numSolverIterationsPostUpdate=4
    numSolverIterationsPreUpdate=16
    alphaBoolBlend {blendOption=Linear}
    boundBone {boneName=dyn_skirt_bk_l}
    chainEnd {boneName=dyn_skirt_bk_02_l}
    physicsBodyDefinitions [{
      sphereCollisionRadius=0
      boundBone {boneName=dyn_skirt_bk_l}
      boxExtents=<40 38 15>
      constraintSetup {angularTargetAxis=AxisZ, twistAxis=AxisZ, angularLimitsMax=<10 10 5>, angularLimitsMin=<-10 -20 -5>, angularTarget=<0 0 1>}
      localJointOffset=<0 0 -3.5>
     }]
    relativeSpaceBone {boneName=pelvis}
   }
   n10 = Flap Node - Bone: dyn_skirt_bk_l!(ComponentPose:Component Space Pose Link Structure=@n9, Alpha:float=1) -> (Pose:Component Space Pose Link Structure) {
    is=title FortAnimGraphNode_Flap
    targetComponent=ScaleZ
    alphaBoolBlend {blendOption=Linear}
    sourceDrives [{
      bUseRange=true
      drivingCurve=null
      multiplier=1
      rangeMax=-0.53
      rangeMin=-2.094
      remappedMax=0
      remappedMin=-0.3
      sourceComponent=RotationZ
      useQuaternion=true
      sourceBone {boneName=calf_l}
     }]
    targetBone {boneName=dyn_skirt_bk_l}
   }
   n11 = "Anim Dynamics - Chain: dyn_skirt_out_r -> dyn_skirt_out_02_r"(ComponentPose=@n10, Alpha=1, ExternalForce="0,0,0") {
    angularDampingOverride=0.999
    angularSpringConstant=100
    bAngularSpring=true
    bChain=true
    bGravityOverrideInSimSpace=true
    bOverrideAngularDamping=true
    numSolverIterationsPostUpdate=4
    numSolverIterationsPreUpdate=16
    alphaBoolBlend {blendOption=Linear}
    boundBone {boneName=dyn_skirt_out_r}
    chainEnd {boneName=dyn_skirt_out_02_r}
    physicsBodyDefinitions [{
      sphereCollisionRadius=0
      boundBone {boneName=dyn_skirt_out_r}
      boxExtents=<50 50 30>
      constraintSetup {angularTargetAxis=AxisZ, twistAxis=AxisZ, angularLimitsMax=<10 0 5>, angularLimitsMin=<-10 -10 -5>, angularTarget=<0 0 1>}
      localJointOffset=<0 0 2>
     }]
    relativeSpaceBone {boneName=pelvis}
   }
   n12 = "Anim Dynamics - Chain: dyn_skirt_fr_r -> dyn_skirt_fr_02_r"(ComponentPose=@n11, Alpha=1, ExternalForce="0,0,0") {
    angularDampingOverride=0.999
    angularSpringConstant=300
    bAngularSpring=true
    bChain=true
    bGravityOverrideInSimSpace=true
    bOverrideAngularDamping=true
    numSolverIterationsPostUpdate=4
    numSolverIterationsPreUpdate=16
    alphaBoolBlend {blendOption=Linear}
    boundBone {boneName=dyn_skirt_fr_r}
    chainEnd {boneName=dyn_skirt_fr_02_r}
    physicsBodyDefinitions [{
      sphereCollisionRadius=0
      boundBone {boneName=dyn_skirt_fr_r}
      boxExtents=<26 22 8>
      constraintSetup {angularTargetAxis=AxisZ, twistAxis=AxisZ, angularLimitsMax=<2 5 5>, angularLimitsMin=<-2 -20 -5>, angularTarget=<0 0 1>}
      localJointOffset=<0 0 5>
     }]
    relativeSpaceBone {boneName=pelvis}
   }
   n13 = "Anim Dynamics - Chain: dyn_skirt_bk_r -> dyn_skirt_bk_02_r"(ComponentPose=@n12, Alpha=1, ExternalForce="0,0,0") {
    angularDampingOverride=0.999
    angularSpringConstant=300
    bAngularSpring=true
    bChain=true
    bGravityOverrideInSimSpace=true
    bOverrideAngularDamping=true
    numSolverIterationsPostUpdate=4
    numSolverIterationsPreUpdate=16
    alphaBoolBlend {blendOption=Linear}
    boundBone {boneName=dyn_skirt_bk_r}
    chainEnd {boneName=dyn_skirt_bk_02_r}
    physicsBodyDefinitions [{
      sphereCollisionRadius=0
      boundBone {boneName=dyn_skirt_bk_r}
      boxExtents=<40 36 15>
      constraintSetup {angularTargetAxis=AxisZ, twistAxis=AxisZ, angularLimitsMax=<0 10 5>, angularLimitsMin=<0 -20 -5>, angularTarget=<0 0 1>}
      localJointOffset=<0 0 3.5>
     }]
    relativeSpaceBone {boneName=pelvis}
   }
   n14 = Flap Node - Bone: dyn_skirt_bk_r!(ComponentPose:Component Space Pose Link Structure=@n13, Alpha:float=1) -> (Pose:Component Space Pose Link Structure) {
    is=title FortAnimGraphNode_Flap
    targetComponent=ScaleZ
    alphaBoolBlend {blendOption=Linear}
    sourceDrives [{
      bUseRange=true
      drivingCurve=null
      multiplier=1
      rangeMax=-0.53
      rangeMin=-2.094
      remappedMax=0
      remappedMin=-0.5
      sourceComponent=RotationZ
      useQuaternion=true
      sourceBone {boneName=calf_r}
     }]
    targetBone {boneName=dyn_skirt_bk_r}
   }
   Output Pose(Result=Component To Local(ComponentPose=@n14)) {layerGroup=None}
   Sequence Player 'Running_Around_Ingame'() {sequence=/Game/Characters/Player/Male/Medium/Bodies/M_MED_Viking_01/Animations/Running_Around_Ingame}
   Flap Node - Bone: dyn_skirt_bk_02_r!(ComponentPose:Component Space Pose Link Structure, Alpha:float=1) -> (Pose:Component Space Pose Link Structure) {
    is=title FortAnimGraphNode_Flap
    targetComponent=RotationY
    alphaBoolBlend {blendOption=Linear}
    sourceDrives [{
      bUseRange=true
      drivingCurve=null
      multiplier=1
      rangeMax=-0.53
      rangeMin=-2.094
      remappedMax=0
      remappedMin=-1.99
      sourceComponent=RotationZ
      useQuaternion=true
      sourceBone {boneName=calf_r}
     }]
    targetBone {boneName=dyn_skirt_bk_02_r}
   }
   Flap Node - Bone: dyn_skirt_bk_02_l!(ComponentPose:Component Space Pose Link Structure, Alpha:float=1) -> (Pose:Component Space Pose Link Structure) {
    is=title FortAnimGraphNode_Flap
    targetComponent=RotationY
    alphaBoolBlend {blendOption=Linear}
    sourceDrives [{
      bUseRange=true
      drivingCurve=null
      multiplier=1
      rangeMax=-0.53
      rangeMin=-2.094
      remappedMax=0
      remappedMin=-1.919
      sourceComponent=RotationZ
      useQuaternion=true
      sourceBone {boneName=calf_l}
     }]
    targetBone {boneName=dyn_skirt_bk_02_l}
   }
  ]}
 EventGraph {nodes [Event BlueprintUpdateAnimation()]}
]
variables [MeshToCopy {type=SkeletalMeshComponent}]
